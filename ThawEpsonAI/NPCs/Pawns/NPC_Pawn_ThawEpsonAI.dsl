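; ThawEpson NPC pawn: a color-variant enemy that switches between idle, hostile, frightened and eye-mode states.
; AI goal changes set hostility, and a gameplay tag triggers fright, which ends after a delay without a goal.
; Lethal damage is intercepted for a timed eye mode with immunity effect and optional loot, then health restores.
; RepNotifies drive material, light, Niagara and audio feedback, while clients turn the mesh toward local players on a timer.
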(event EventBeginPlay
  (|CallParentFunction)
  (CallFunction|UpdateSoundLibraryAudio "(TagName=\"SoundLibrary.ThawEpsonAI.Loop\")" true)
  (CallFunction|CheckFrightenedStatus)
  (if (ScalableFloat|GetValueasBool (Variables|Default|GetSF_HideHealthbar))
    (CallFunction|NPCDisableHealthBars)
    (else
      _))
  (Networking|SwitchHasAuthority
    (:Authority
      (CallFunction|InitPawn)
      (Pawn|BindEventtoReceiveControllerChangedDelegate (delegate Custom|ControllerReady))
      (Ability|Async|WaitGameplayEventToActor 0 "(TagName=\"Event.CosmicThunder.ThawEpsonAI.MeleeAttack\")"
        (:EventReceived
          (CallFunction|HandleAttacking))))
    (:Remote
      (CallFunction|StartFacingLogic)
      ;[1 1 1] WashForceAndDustFX
      (Utilities|EffectsQualityCheck 0
        (:Cinematic
          _)
        (:Epic
          _))
      (Variables|Default|GetPhysicalForceActor
        (:then)
        (:else
          (Utilities|FlowControl|DelayUntilNextTick)
          (if (not (FortPawn|IsDead))
            (bind _returnvalue (Game|SpawnActorfromClass "/Game/Effects/Niagara/Blueprints/PhsyicsFields/FN_WashForce.FN_WashForce_C" (Transformation|GetActorTransform) "Undefined" "MultiplyWithRoot" 0 7e+02 0.0 -999.0 0))
            (Variables|Default|SetPhysicalForceActor _returnvalue)
            (Transformation|AttachActorToActor (Variables|Default|GetPhysicalForceActor) self "None" "SnapToTarget" "KeepRelative" "KeepRelative" true)))))))

(event Custom|ControllerReady
  (params (Pawn :type Pawn%20Object%20Reference) (OldController :type Controller%20Object%20Reference) (NewController :type Controller%20Object%20Reference))
  (CallFunction|CheckGoal))

(event Custom|StartFacingLogic
  ; Make Pawn Look at PLayers
  (bind _returnvalue (Utilities|Time|SetTimerbyEvent (EventDispatchers|CreateEvent :!event Custom|UpdateMeshOrientation) (ScalableFloat|GetValueatLevel (Variables|Default|GetFacingInterval)) true))
  (Variables|InternalData|SetHNDL_OrientationTimer _returnvalue))

(event Custom|CE_SpawnWashForceActor
  ;[1 1 1] WashForceAndDustFX
  (Utilities|EffectsQualityCheck 0
    (:Cinematic
      _)
    (:Epic
      _))
  (Variables|Default|GetPhysicalForceActor
    (:then)
    (:else
      (Utilities|FlowControl|DelayUntilNextTick)
      (if (not (FortPawn|IsDead))
        (bind _returnvalue (Game|SpawnActorfromClass "/Game/Effects/Niagara/Blueprints/PhsyicsFields/FN_WashForce.FN_WashForce_C" (Transformation|GetActorTransform) "Undefined" "MultiplyWithRoot" 0 7e+02 0.0 -999.0 0))
        (Variables|Default|SetPhysicalForceActor _returnvalue)
        (Transformation|AttachActorToActor (Variables|Default|GetPhysicalForceActor) self "None" "SnapToTarget" "KeepRelative" "KeepRelative" true)))))

(event EventEndPlay
  (params (EndPlayReason :type EEndPlayReason%20Enum :default "Destroyed"))
  (|CallParentFunction EndPlayReason)
  (Utilities|Time|ClearandInvalidateTimerbyHandle (Variables|InternalData|GetHNDL_EyeModeTimer))
  (Utilities|Time|ClearandInvalidateTimerbyHandle (Variables|InternalData|GetHNDL_OrientationTimer)))

(event Custom|UpdateMeshOrientation
  (bind _self self)
  (bind _returnvalue (Transformation|GetActorLocation _self))
  (bind (_returnvalue_roll _returnvalue_pitch) (Math|Rotator|BreakRotator (Transformation|GetWorldRotation (Variables|Default|GetSM_ThawEpson))))
  ; Make Pawn Look at PLayers
  (for _array_element (Fortnite|GetLocalFortPlayerControllers true)
    (bind (_returnvalue_roll_2 _returnvalue_pitch_2 _returnvalue_yaw) (Math|Rotator|BreakRotator (Math|Rotator|FindLookatRotation _returnvalue (Transformation|GetActorLocation (Pawn|GetControlledPawn _array_element)))))
    (Transformation|SetWorldRotation (Variables|Default|GetSM_ThawEpson) (Math|Rotator|MakeRotator _returnvalue_roll _returnvalue_pitch (+ _returnvalue_yaw -9e+01)) true false)
    ;; Update Eye Location
    (if (< (Math|Vector|DotProduct (Math|Vector|Normalize (- _returnvalue (Camera|GetCameraLocation (Variables|/Script/Engine.PlayerController|GetPlayerCameraManager _array_element)))) (Math|Vector|GetRightVector (Camera|GetCameraRotation (Variables|/Script/Engine.PlayerController|GetPlayerCameraManager _array_element)))) 0)
      (Rendering|Material|SetScalarParameterValue (Variables|InternalData|GetMatRef) "ThawEpson_Eyes_Left_Right" 1.0 :!class /Script/Engine.MaterialInstanceDynamic)
      (else
        (Rendering|Material|SetScalarParameterValue (Variables|InternalData|GetMatRef) "ThawEpson_Eyes_Left_Right" 0.0 :!class /Script/Engine.MaterialInstanceDynamic)))))

(event Custom|CE_DestroyWashForceActor
  ;[1 1 1] WashForceAndDustFX
  (bind _physicalforceactor (Variables|Default|GetPhysicalForceActor)
    (:then
      (Actor|DestroyActor _physicalforceactor))
    (:else)))

(event EventOnDeathServer
  (params (Damage :type Float%20%28single-precision%29) (DamageTags :type Gameplay%20Tag%20Container%20Structure%20%28by%20ref%29) (Momentum :type Vector :default "0, 0, 0") (HitInfo :type Hit%20Result%20Structure%20%28by%20ref%29) (InstigatedBy :type Controller%20Object%20Reference) (DamageCauser :type Actor%20Object%20Reference) (EffectContext :type Gameplay%20Effect%20Context%20Handle%20Structure))
  (|CallParentFunction Damage DamageTags Momentum HitInfo InstigatedBy DamageCauser EffectContext)
  (CallFunction|CEDestroy)
  (if (GameplayTags|HasTag DamageTags "(TagName=\"Gameplay.Damage.OutsideSafeZone\")")
    (Variables|Default|SetDeathByOutsideSafeZone true)))

(event Custom|CheckGoal
  ; Check Goal to Set Hostility
  (bind _asaicontroller (Utilities|Casting|CastToAIController (Pawn|GetController)))
  (bind _returnvalue (AI|GoalSelection|GetAIGoalComponent _asaicontroller))
  (EventDispatchers|BindEventtoOnNewGoalActorDelegate _returnvalue (EventDispatchers|CreateEvent :!event Custom|OnGoalUpdate)))

(event Custom|CE_Destroy
  (Networking|CanExecuteCosmeticEvents
    (:True
      (CallFunction|CEDestroyWashForceActor))))

(event Custom|HandleAttacking
  ; Update Mesh to Trigger Attack VFX
  (Variables|InternalData|SetIsAttacking true)
  (Utilities|FlowControl|Delay 0.2)
  (Variables|InternalData|SetIsAttacking false))

(event Custom|OnGoalUpdate
  (params (OldActor :type Actor%20Object%20Reference) (NewActor :type Actor%20Object%20Reference))
  ; Check Goal to Set Hostility
  (Variables|InternalData|SetGoalActor NewActor)
  (if (Math|Boolean|ANDBoolean (!= NewActor 0) (not (Variables|Default|GetIsFrightened)) (not (Variables|Default|GetIsEyeMode)))
    (Variables|Default|SetIsHostile true)
    (else
      (Variables|Default|SetIsHostile false))))

(event EventOnDeathPlayEffects
  (params (Damage :type Float%20%28single-precision%29) (DamageTags :type Gameplay%20Tag%20Container%20Structure%20%28by%20ref%29) (Momentum :type Vector :default "0, 0, 0") (HitInfo :type Hit%20Result%20Structure%20%28by%20ref%29) (InstigatedBy :type Fort%20Pawn%20Object%20Reference) (DamageCauser :type Actor%20Object%20Reference) (EffectContext :type Gameplay%20Effect%20Context%20Handle%20Structure))
  ;[1 1 1] On Death Effects
  (|CallParentFunction Damage DamageTags Momentum HitInfo InstigatedBy DamageCauser EffectContext)
  (CallFunction|UpdateSoundLibraryAudio "(TagName=\"SoundLibrary.ThawEpsonAI.Loop\")")
  (CallFunction|SpawnDeathEffects)
  (CallFunction|UpdateSoundLibraryAudio "(TagName=\"SoundLibrary.ThawEpsonAI.Death\")" true))

(event Custom|CheckFrightenedStatus
  ; Check Frightened Status
  (Ability|Tasks|WaitGameplayTagAddToActor 0 (Variables|Default|GetTag_Frightened)
    (:Added
      (Variables|Default|SetIsFrightened true))))

(event Custom|SpawnDeathEffects
  ;[1 1 1] On Death Effects
  (CallFunction|NPCDisableHealthBars))

(event Custom|PawnShouldHaveDied
  ; Handle Eye State
  (Utilities|FlowControl|DoOnce false
    (:Completed
      (if (not (Variables|Default|GetIsFrightened))
        (Ability|SendGameplayEventtoActor self "(TagName=\"Event.CosmicThunder.ThawEpsonAI.ForceFrightened\")") ; Too much damage was delt too fast, AI should have this tag!\nUsed in HealthCheck Ability
      )
      (CallFunction|NPCDisableHealthBars)
      (FortPawn|SetHealth self 1.0)
      (Variables|Default|SetIsEyeMode true)
      (bind _returnvalue (Utilities|Time|SetTimerbyEvent (EventDispatchers|CreateEvent :!event Custom|ReturnToNormal) (ScalableFloat|GetValueatLevel (Variables|Default|GetEyeModeDuration))))
      (Variables|InternalData|SetHNDL_EyeModeTimer _returnvalue)
      (Environment|ApplyGlobalEnvironmentGameplayEffect (Variables|Default|GetGE_IsEyesMode))
      (CallFunction|TriggerEyeMode)
      (if (ScalableFloat|GetValueasBool (Variables|Default|GetShouldDropLootOnEyeMode))
        (Inventory|SpawnPickupInWorldWithLootTier :LootTierName "Loot_ApolloTreasure_Rare" :Position (+ (Transformation|GetActorLocation) "0.000000,0.000000,50.000000") :OverrideMaxStackCount 0 :bToss true :bTossWithVelocity false :SourceType "Tossed" :Source "NPCElimination" :LootWorldLevel 1)))))

(event Custom|ReturnToNormal
  ; Handle Eye State
  (FortPawn|SetHealth (FortPawn|GetMaxHealth))
  (Variables|Default|SetIsFrightened false)
  (Variables|Default|SetIsEyeMode false)
  (Default|CallOnReturnedtoNormal)
  (CallFunction|NPCEnableHealthBars)
  (CallFunction|TriggerRespawnGC)
  (Utilities|FlowControl|DoOnce false
    (:Completed
      (if (not (Variables|Default|GetIsFrightened))
        (Ability|SendGameplayEventtoActor self "(TagName=\"Event.CosmicThunder.ThawEpsonAI.ForceFrightened\")") ; Too much damage was delt too fast, AI should have this tag!\nUsed in HealthCheck Ability
      )
      (CallFunction|NPCDisableHealthBars)
      (FortPawn|SetHealth self 1.0)
      (Variables|Default|SetIsEyeMode true)
      (bind _returnvalue (Utilities|Time|SetTimerbyEvent (EventDispatchers|CreateEvent :!event Custom|ReturnToNormal) (ScalableFloat|GetValueatLevel (Variables|Default|GetEyeModeDuration))))
      (Variables|InternalData|SetHNDL_EyeModeTimer _returnvalue)
      (Environment|ApplyGlobalEnvironmentGameplayEffect (Variables|Default|GetGE_IsEyesMode))
      (CallFunction|TriggerEyeMode)
      (if (ScalableFloat|GetValueasBool (Variables|Default|GetShouldDropLootOnEyeMode))
        (Inventory|SpawnPickupInWorldWithLootTier :LootTierName "Loot_ApolloTreasure_Rare" :Position (+ (Transformation|GetActorLocation) "0.000000,0.000000,50.000000") :OverrideMaxStackCount 0 :bToss true :bTossWithVelocity false :SourceType "Tossed" :Source "NPCElimination" :LootWorldLevel 1)))))

(event EventOnDamagePlayEffects
  (params (Damage :type Float%20%28single-precision%29) (DamageTags :type Gameplay%20Tag%20Container%20Structure%20%28by%20ref%29) (Momentum :type Vector :default "0, 0, 0") (HitInfo :type Hit%20Result%20Structure%20%28by%20ref%29) (InstigatedBy :type Fort%20Pawn%20Object%20Reference) (DamageCauser :type Actor%20Object%20Reference) (EffectContext :type Gameplay%20Effect%20Context%20Handle%20Structure))
  (bind (_bblockinghit _binitialoverlap _time _distance _location _impactpoint _normal _impactnormal) (Collision|BreakHitResult HitInfo))
  ;[1 1 1] On Damaged
  (|CallParentFunction Damage DamageTags Momentum HitInfo InstigatedBy DamageCauser EffectContext)
  (bind _returnvalue (Niagara|SpawnSystematLocation "/ThawEpsonAI/Assets/VFX/NS_ThawEpson_DamageImpact.NS_ThawEpson_DamageImpact" _impactpoint (Math|Conversions|RotationFromXVector _impactnormal)))
  (Niagara|SetNiagaraVariable(Float) _returnvalue "Color ID" (select (Variables|Default|GetIsFrightened) 4.0 (Variables|Default|GetColorID)))
  (Audio|PlaySoundatLocation "/ThawEpsonAI/Assets/Sounds/IP_SC_ThawEpson_HitDamage.IP_SC_ThawEpson_HitDamage" (Transformation|GetActorLocation)))

(event Custom|TriggerEyeMode
  ; Handle Eye State
  (GameplayCue|ExecuteGameplayCue%28Burst%29Local self (Variables|Default|GetGC_EyeModeTriggered) (Ability|GameplayCue|MakeGameplayCueParameters :NormalizedMagnitude 0.0 :RawMagnitude 0.0 :OriginalTag (Variables|Default|GetGC_EyeModeTriggered) :Location "0,0,0" :Normal "0,0,0" :GameplayEffectLevel 1 :AbilityLevel 1 :bReplicateLocationWhenUsingMinimalRepProxy false))
  (Audio|PlaySoundatLocation "/ThawEpsonAI/Assets/Sounds/IP_SC_ThawEpson_HitDeath.IP_SC_ThawEpson_HitDeath" (Transformation|GetActorLocation)))

(event Custom|TriggerRespawnGC
  ; Handle Eye State
  (GameplayCue|ExecuteGameplayCue%28Burst%29Local self (Variables|Default|GetGC_Respawn) (Ability|GameplayCue|MakeGameplayCueParameters :NormalizedMagnitude 0.0 :RawMagnitude (Variables|Default|GetColorID) :OriginalTag (Variables|Default|GetGC_Respawn) :Location "0,0,0" :Normal "0,0,0" :GameplayEffectLevel 1 :AbilityLevel 1 :bReplicateLocationWhenUsingMinimalRepProxy false)))

(event Custom|InsectSwarm_DisplayLeash
  (params (Duration :type Float%20%28double-precision%29))
  (bind _randomdebugcolorselected (Variables|Debug|GetRandomDebugColorSelected))
  (bind _radiusdebugcolorarray (Variables|Debug|GetRadiusDebugColorArray))
  (bind _output (Utilities|Array|Get(aref) _radiusdebugcolorarray _randomdebugcolorselected))
  (bind _leashlocation (Variables|InternalData|GetLeashLocation))
  ;[1 1 1] Draw Debug Data for Swarm
  (Variables|Debug|SetRandomDebugColorSelected (Math|Random|RandomIntegerinRange 0 (- (Utilities|Array|Length (Variables|Debug|GetRadiusDebugColorArray)) 1)))
  (Rendering|Debug|DrawDebugSphere _leashlocation (ScalableFloat|GetValueatLevel (Variables|Default|GetLeashRadius)) 12 _output Duration)
  (Rendering|Debug|DrawDebugSphere _leashlocation 4e+01 12 _output Duration)
  (Rendering|Debug|DrawDebugString (+ _leashlocation "0.000000,0.000000,45.000000") (Utilities|String|Append "Center of leash for: " (Utilities|GetDisplayName self)) 0 _output Duration))

(event Custom|HandleLeaveFrightened
  ; Check if we should revert back to normal state after we started being frightened
  (bind _returnvalue (Utilities|Time|SetTimerbyEvent (delegate Custom|CheckShouldLeaveFrightened) (ScalableFloat|GetValueatLevel (Variables|Default|GetLeaveFrightenedInterval)) true))
  (Variables|InternalData|SetHNDL_LeaveFrightenedTimer _returnvalue))

(event Custom|CheckShouldLeaveFrightened
  ; Check if we should revert back to normal state after we started being frightened
  (if (or (Variables|Default|GetIsEyeMode) (not (Variables|Default|GetIsFrightened)))
    (Utilities|Time|ClearandInvalidateTimerbyHandle (Variables|InternalData|GetHNDL_LeaveFrightenedTimer))
    (else
      (if (and (>= (- (Utilities|Time|GetGameTimeinSeconds) (Variables|InternalData|GetLastFrightenedStartTime)) (ScalableFloat|GetValueatLevel (Variables|Default|GetLeaveFrightenedDelay))) (not (Utilities|IsValid (Variables|InternalData|GetGoalActor) :!class /Script/Engine.KismetSystemLibrary)))
        (Utilities|Time|ClearandInvalidateTimerbyHandle (Variables|InternalData|GetHNDL_LeaveFrightenedTimer))
        (CallFunction|ReturntoNormal)))))

(event Custom|HandleLeashPositionUpdate
  (bind _self self)
  (bind _returnvalue (Pawn|GetController _self))
  (bind _returnvalue_2 (AI|GetBlackboard _returnvalue))
  (bind _goalactorbbname (Variables|InternalData|GetGoalActorBBNAme))
  (bind _returnvalue_3 (AI|Components|Blackboard|GetValueasObject _returnvalue_2 _goalactorbbname))
  ;[1 1 1] Update Leash Position when Swarm gets a new Target so we reduce chances that a Swarm can be at the edge of his leash and then chaces a player for just a second of two.
  (Utilities|IsValid _returnvalue
    (:Is%20Valid
      (Utilities|IsValid _returnvalue_2
        (:Is%20Valid
          (Utilities|IsValid _returnvalue_3
            (:Is%20Valid
              (if (not (== (Variables|InternalData|GetTargetActor) _returnvalue_3))
                (bind _asactor (Utilities|Casting|CastToActor _returnvalue_3))
                (Variables|InternalData|SetLeashLocation (Transformation|GetActorLocation self))
                (Variables|InternalData|SetTargetActor _asactor)))))))))

(fn BlueprintModifyIncomingDamage
  (params (Damage :type Float%20%28single-precision%29) (DamageTags :type Gameplay%20Tag%20Container%20Structure%20%28by%20ref%29) (EffectContext :type Gameplay%20Effect%20Context%20Handle%20Structure%20%28by%20ref%29) (EventInstigator :type Controller%20Object%20Reference) (DamageCauser :type Actor%20Object%20Reference))
  (bind _returnvalue (|CallParentFunction Damage DamageTags EffectContext EventInstigator DamageCauser))
  (if (<= (- (FortPawn|GetHealth self) Damage) 0)
    (CallFunction|PawnShouldHaveDied)
    (return 0.0)
    (else
      (return _returnvalue))))

(fn InitPawn
  (Variables|InternalData|SetLeashLocation (Transformation|GetActorLocation self))
  (return))

(fn IsInsideLeashRadius
  (params (ActorToTest :type Actor%20Object%20Reference) (VectorToTest :type Vector))
  (Utilities|IsValid ActorToTest
    (:Is%20Valid
      (LocalVariables|SetLocationToTest (Transformation|GetActorLocation ActorToTest)))
    (:Is%20Not%20Valid
      (LocalVariables|SetLocationToTest VectorToTest)))
  (return (<= (Math|Vector|Distance2D(Vector) (LocalVariables|GetLocationToTest) (Variables|InternalData|GetLeashLocation)) (ScalableFloat|GetValueatLevel (Variables|Default|GetLeashRadius)))))

(fn IsValidAutoFireTarget
  (params (Attacker :type Actor%20Object%20Reference))
  (return true))

(fn OnRep_DeathByOutsideSafeZone)

(fn OnRep_InteractingPawn)

(fn OnRep_IsAttacking
  (Networking|SwitchHasAuthority
    (:Remote
      (if (Variables|InternalData|GetIsAttacking)
        (Rendering|Material|SetScalarParameterValue (Variables|InternalData|GetMatRef) "ThawEpson_StartTime_Attack" (Utilities|Time|GetGameTimeinSeconds) :!class /Script/Engine.MaterialInstanceDynamic)
        (Audio|PlaySoundatLocation "/ThawEpsonAI/Assets/Sounds/IP_SC_ThawEpson_Attack.IP_SC_ThawEpson_Attack" (Transformation|GetActorLocation))))))

(fn OnRep_IsEyeMode
  (bind _self self)
  (bind _returnvalue (Ability|GetAbilitySystemComponent _self))
  (bind _weaponcapsulecollision (Variables|Default|GetWeaponCapsuleCollision))
  (if (Variables|Default|GetIsEyeMode)
    (Collision|SetCollisionEnabled _weaponcapsulecollision "NoCollision")
    (else
      (Collision|SetCollisionEnabled _weaponcapsulecollision "QueryOnly")))
  (Networking|SwitchHasAuthority
    (:Authority
      (Utilities|FlowControl|Sequence
        (:then_0
          ; Apply IsEyeMode GE
          (if (Variables|Default|GetIsEyeMode)
            (Utilities|IsValid _returnvalue
              (:Is%20Valid
                (bind _returnvalue_2 (GameplayEffects|ApplyGameplayEffectToTarget :self _returnvalue :GameplayEffectClass (Variables|Default|GetGE_IsEyesMode) :Target _returnvalue :Level 0.0))
                (Variables|InternalData|SetHNDL_ImmunityGE _returnvalue_2)))
            (else
              (Utilities|IsValid _returnvalue
                (:Is%20Valid
                  (GameplayEffects|RemoveActiveGameplayEffect _returnvalue (Variables|InternalData|GetHNDL_ImmunityGE) -1))))))
        (:then_1
          _)))
    (:Remote
      _))
  (if (Variables|Default|GetIsEyeMode)
    (Rendering|Material|SetScalarParameterValue (Variables|InternalData|GetMatRef) "ThawEpson_IsScared" 0.0 :!class /Script/Engine.MaterialInstanceDynamic)
    (Rendering|Material|SetScalarParameterValue (Variables|InternalData|GetMatRef) "ThawEpson_IsDead" 1.0 :!class /Script/Engine.MaterialInstanceDynamic)
    (Components|Activation|SetActive (Variables|Default|GetNS_ThawEpson_Idle) false false)
    (Rendering|Components|Light|SetIntensity (Variables|Default|GetBR_Pointlight_Component) 5.0)
    (CallFunction|UpdateSoundLibraryAudio "(TagName=\"SoundLibrary.ThawEpsonAI.Death\")" true)
    (else
      (Components|Activation|SetActive (Variables|Default|GetNS_ThawEpson_Idle) true false)
      (if (Variables|Default|GetIsFrightened)
        (Rendering|Material|SetScalarParameterValue (Variables|InternalData|GetMatRef) "ThawEpson_IsDead" 0.0 :!class /Script/Engine.MaterialInstanceDynamic)
        (Rendering|Material|SetScalarParameterValue (Variables|InternalData|GetMatRef) "ThawEpson_IsScared" 1.0 :!class /Script/Engine.MaterialInstanceDynamic)
        (else
          (Rendering|Material|SetScalarParameterValue (Variables|InternalData|GetMatRef) "ThawEpson_IsDead" 0.0 :!class /Script/Engine.MaterialInstanceDynamic)
          (Rendering|Material|SetScalarParameterValue (Variables|InternalData|GetMatRef) "ThawEpson_IsScared" 0.0 :!class /Script/Engine.MaterialInstanceDynamic)
          (Rendering|Components|Light|SetIntensity (Variables|Default|GetBR_Pointlight_Component) 15.0)
          (Rendering|Components|Light|SetLightColor (Variables|Default|GetBR_Pointlight_Component) (Utilities|Array|Get(acopy) (Variables|Default|GetLightColors) (Math|Float|Truncate (Variables|Default|GetColorID))) true)
          (if (Variables|Default|GetIsHostile)
            (Rendering|Material|SetScalarParameterValue (Variables|InternalData|GetMatRef) "ThawEpson_StartTime_Aggro" (Utilities|Time|GetGameTimeinSeconds) :!class /Script/Engine.MaterialInstanceDynamic)
            (CallFunction|UpdateSoundLibraryAudio "(TagName=\"SoundLibrary.ThawEpsonAI.Alert\")" true)
            (elif (== (Variables|InternalData|GetGoalActor) 0)
              (CallFunction|UpdateSoundLibraryAudio "(TagName=\"SoundLibrary.ThawEpsonAI.Loop\")" true)
              (else
                (Variables|Default|SetIsHostile true)
                (Rendering|Material|SetScalarParameterValue (Variables|InternalData|GetMatRef) "ThawEpson_StartTime_Aggro" (Utilities|Time|GetGameTimeinSeconds) :!class /Script/Engine.MaterialInstanceDynamic)
                (CallFunction|UpdateSoundLibraryAudio "(TagName=\"SoundLibrary.ThawEpsonAI.Alert\")" true)))))))))

(fn OnRep_IsFrightened
  (bind (_bisvalid _returnvalue) (Niagara|GetNiagaraVariable(Int32) (Variables|Default|GetNS_ThawEpson_Idle) "Color Array"))
  (Networking|SwitchHasAuthority
    (:Authority
      (if (Variables|Default|GetIsFrightened)
        (Variables|InternalData|SetLastFrightenedStartTime (Utilities|Time|GetGameTimeinSeconds))
        (CallFunction|HandleLeaveFrightened)
        (else
          (Utilities|Time|ClearandInvalidateTimerbyHandle (Variables|InternalData|GetHNDL_LeaveFrightenedTimer)))))
    (:Remote
      (if (Variables|Default|GetIsEyeMode)
        _
        (elif (Variables|Default|GetIsFrightened)
          (CallFunction|UpdateSoundLibraryAudio "(TagName=\"SoundLibrary.ThawEpsonAI.Threatened\")" true)
          (LocalVariables|SetFrightenedValue 1.0)
          (if (not (Variables|Default|GetFrightenedColorUpdated))
            (Variables|Default|SetFrightenedColorUpdated true)
            (Variables|InternalData|SetOGColorValue _returnvalue)
            (Niagara|SetNiagaraVariableByString(Int32) (Variables|Default|GetNS_ThawEpson_Idle) "Color Array" 4)
            (Rendering|Material|SetScalarParameterValue (Variables|InternalData|GetMatRef) "ThawEpson_IsScared" (LocalVariables|GetFrightenedValue) :!class /Script/Engine.MaterialInstanceDynamic)
            (Rendering|Components|Light|SetLightColor (Variables|Default|GetBR_Pointlight_Component) "(R=0.015209,G=0.015209,B=1.000000,A=1.000000)" true))
          (else
            (Variables|Default|SetFrightenedColorUpdated false)
            (LocalVariables|SetFrightenedValue 0.0)
            (Niagara|SetNiagaraVariableByString(Int32) (Variables|Default|GetNS_ThawEpson_Idle) "Color Array" (Variables|InternalData|GetOGColorValue))))))))

(fn OnRep_IsHostile
  (Networking|SwitchHasAuthority
    (:Remote
      (if (Variables|Default|GetIsHostile)
        (Rendering|Material|SetScalarParameterValue (Variables|InternalData|GetMatRef) "ThawEpson_StartTime_Aggro" (Utilities|Time|GetGameTimeinSeconds) :!class /Script/Engine.MaterialInstanceDynamic)
        (CallFunction|UpdateSoundLibraryAudio "(TagName=\"SoundLibrary.ThawEpsonAI.Alert\")" true)
        (elif (Variables|Default|GetIsEyeMode)
          _
          (else
            (Rendering|Material|SetScalarParameterValue (Variables|InternalData|GetMatRef) "ThawEpson_IsDead" 0.0 :!class /Script/Engine.MaterialInstanceDynamic)
            (CallFunction|UpdateSoundLibraryAudio "(TagName=\"SoundLibrary.ThawEpsonAI.Loop\")" true)))))))

(fn UpdateSoundLibraryAudio
  (params (InEventName :type Gameplay%20Tag%20Structure) (Play :type Boolean))
  (bind _thawepsonai_soundlibrarycomponent (Variables|Default|GetThawEpsonAI_SoundLibraryComponent))
  (if Play
    (SoundLibrary|PlayEvent _thawepsonai_soundlibrarycomponent InEventName)
    (else
      (SoundLibrary|StopEvent _thawepsonai_soundlibrarycomponent InEventName))))

(fn ConstructionScript
  (|CallParentFunction)
  (bind _returnvalue (Rendering|Material|CreateDynamicMaterialInstance (Variables|Default|GetSM_ThawEpson) 0 (Rendering|Material|GetMaterial (Variables|Default|GetSM_ThawEpson) 0) "None" :!class /Script/Engine.PrimitiveComponent))
  (bind _output_get (Variables|InternalData|SetMatRef _returnvalue))
  (Rendering|Material|SetScalarParameterValue _output_get "ThawEpson_ColorID" (Variables|Default|GetColorID) :!class /Script/Engine.MaterialInstanceDynamic)
  (bind _returnvalue_2 (Rendering|Material|CreateDynamicMaterialInstance (Variables|Default|GetSM_ThawEpson) 1 (Rendering|Material|GetMaterial (Variables|Default|GetSM_ThawEpson) 1) "None" :!class /Script/Engine.PrimitiveComponent))
  (Variables|Default|SetBlackMatRef _returnvalue_2)
  (Rendering|Components|Light|SetLightColor (Variables|Default|GetBR_Pointlight_Component) (Utilities|Array|Get(acopy) (Variables|Default|GetLightColors) (Math|Float|Truncate (Variables|Default|GetColorID))) true))
